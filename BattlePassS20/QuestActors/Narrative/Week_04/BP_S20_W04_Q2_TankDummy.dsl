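; Season 20 Week 4 quest tank dummy that acts as a bounce pad for overlapping actors.
; Four box overlaps feed BounceActor, which launches pawns upward with a gameplay effect, fall-damage immunity, sound and force feedback, and redirects eligible projectiles or launches hoverboards.
; A capsule-trace ceiling check lowers launch height under obstructions.
; Collision and launch details go to screen and shipping log; the wobble effect event is empty.
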(event Custom|PlayBounceSound
  ;[1 1 1] Keeps bounce sound from double triggering (hopefully)
  (Utilities|FlowControl|DoOnce false
    (:Completed
      (Audio|PlaySoundatLocation (Variables|Default|GetTireSound) (Transformation|GetActorLocation (Variables|Default|GetOverlapPawn))))))

(event Custom|Reset%20Bounce%20Sound
  ;[1 1 1] Keeps bounce sound from double triggering (hopefully)
  (Utilities|FlowControl|DoOnce false
    (:Completed
      (Audio|PlaySoundatLocation (Variables|Default|GetTireSound) (Transformation|GetActorLocation (Variables|Default|GetOverlapPawn))))))

(event EventBeginPlay
  (|CallParentFunction)
  (CallFunction|CollisionLogging "Collision Check - BeginPlay:" (Variables|Default|GetBoxCollision_BounceTank))
  (CallFunction|CollisionLogging "Collision Check - BeginPlay:" (Variables|Default|GetBoxCollision_BounceBarrel))
  (CallFunction|CollisionLogging "Collision Check - BeginPlay:" (Variables|Default|GetBoxCollision_BounceTurret))
  (CallFunction|CollisionLogging "Collision Check - BeginPlay:" (Variables|Default|GetBoxCollision_BounceBarrelTop)))

(event Custom|MakeWobbleEffect
  (params (Center :type Vector)))

(event Custom|BounceActor
  (params (OverlappedComponent :type Primitive%20Component%20Object%20Reference) (OtherActor :type Actor%20Object%20Reference) (SweepResult :type Hit%20Result%20Structure%20%28by%20ref%29))
  (bind _returnvalue (Actor|GetComponentbyClass OtherActor "/Script/Engine.ProjectileMovementComponent"))
  (bind (_bblockinghit _binitialoverlap _time _distance _location _impactpoint) (Collision|BreakHitResult SweepResult))
  (bind (_asfort_projectile_petrol _bsuccess) (Utilities|Casting|CastToFortProjectilePetrol OtherActor))
  (bind (_asfort_pickup _bsuccess_2) (Utilities|Casting|CastToFortPickup OtherActor))
  (bind _overlappawn (Variables|Default|GetOverlapPawn))
  (bind _overlapcontroller (Variables|Default|GetOverlapController))
  (bind _finalbouncejumpheight (Variables|Default|GetFinalBounceJumpHeight))
  (bind _overlappawn_2 (Variables|Default|GetOverlapPawn))
  (bind _overlappawn_3 (Variables|Default|GetOverlapPawn))
  (CallFunction|TankLogging "OnCollision:" OverlappedComponent OtherActor)
  (Variables|Default|SetBounceLocation (Transformation|GetActorLocation))
  (CallFunction|MakeWobbleEffect _impactpoint)
  (bind _asfort_player_pawn (Utilities|Casting|CastToFortPlayerPawn OtherActor)
    (:then
      (Variables|Default|SetOverlapPawn _asfort_player_pawn))
    (:CastFailed
      (bind _asfort_aipawn (Utilities|Casting|CastToFortAIPawn OtherActor)
        (:then
          (Variables|Default|SetOverlapPawn _asfort_aipawn))
        (:CastFailed
          (Utilities|FlowControl|Sequence
            (:then_0
              ;[1 0.03743 0] Launching Projectile
              (if (Math|Boolean|ANDBoolean (or (Utilities|IsValid _returnvalue :!class /Script/Engine.KismetSystemLibrary) (Utilities|IsValid (Actor|GetComponentbyClass OtherActor "/Script/FortniteGame.FortProjectileMovementComponent") :!class /Script/Engine.KismetSystemLibrary)) (not _bsuccess) (not _bsuccess_2) (Variables|Default|GetBBouncingProjectile))
                (bind _returnvalue_2 (Default|GetNonPawnBounceVel OtherActor)
                  (:Output
                    (bind (_returnvalue_x _returnvalue_y _returnvalue_z) (Math|Vector|BreakVector _returnvalue_2))
                    (Variables|/Script/Engine.MovementComponent|SetVelocity (Math|Vector|MakeVector _returnvalue_x _returnvalue_y _returnvalue_z) _returnvalue)))))
            (:then_1
              ;[1 1 1] Ceiling check
              (bind _success (Default|CeilingCheck))
              (if _success
                (Variables|Default|SetBounceHeight 600)
                (Variables|Default|SetJumpHeight 100)
                (else
                  (Variables|Default|SetBounceHeight 2e+03)
                  (Variables|Default|SetJumpHeight 600)))
              (Variables|Default|SetFinalBounceJumpHeight (Variables|Default|GetBounceHeight))
              (bind _success_2 (Vehicles|LaunchHoverboard OtherActor (Math|Vector|MakeVector 0.0 0.0 (Variables|Default|GetFinalBounceJumpHeight)) "TireLaunch"))
              (if _success_2
                (Utilities|FlowControl|Sequence
                  (:then_0
                    (CallFunction|PlayBounceSound))
                  (:then_1
                    ;[1 1 1] Reset Bounce sound, so sound cue can be heard if "X" times bounce happens
                    (Utilities|FlowControl|Delay 0.1)
                    (CallFunction|ResetBounceSound))))))))))
  (Utilities|IsValid _overlappawn_3
    (:Is%20Valid
      (if (or (GameplayTags|HasMatchingGameplayTag (Ability|GetAbilitySystemComponent _overlappawn_3) "(TagName=\"Gameplay.Action.Player.Launched\")") (Variables|/Script/FortniteGame.FortPawn|GetbIsDBNO _overlappawn_3))
        _
        (else
          _)))
    (:Is%20Not%20Valid
      _))
  (bind _asfort_player_controller (Utilities|Casting|CastToFortPlayerController (Pawn|GetController _overlappawn_3))
    (:then
      (Variables|Default|SetOverlapController _asfort_player_controller))
    (:CastFailed
      _))
  ;[1 1 1] Ceiling check
  (bind _success_3 (Default|CeilingCheck))
  (if _success_3
    (Variables|Default|SetBounceHeight 6e+02)
    (Variables|Default|SetJumpHeight 100)
    (else
      (Variables|Default|SetBounceHeight (Variables|Default|GetNoCeilingBounceHeight))
      (Variables|Default|SetJumpHeight 600)))
  (Variables|Default|SetFinalBounceJumpHeight (Variables|Default|GetBounceHeight))
  (bind _asfort_player_pawn_2 (Utilities|Casting|CastToFortPlayerPawn (Variables|Default|GetOverlapPawn))
    (:then
      (RemoteControlPawn|ForceDetachfromRemoteControlledPawn _asfort_player_pawn_2))
    (:CastFailed
      _))
  (Character|LaunchCharacter _overlappawn_2 (Math|Vector|MakeVector 0.0 0.0 _finalbouncejumpheight) false true)
  (CallFunction|TankLogging "OnLaunch:" OverlappedComponent _overlappawn_2 _finalbouncejumpheight)
  (GameplayEffects|ApplyGameplayEffectToSelf :self (Ability|GetAbilitySystemComponent _overlappawn) :GameplayEffectClass (Variables|Default|GetGE_PlayerLaunched) :Level 0.0)
  (Variables|/Script/FortniteGame.FortPawn|SetbIgnoreNextFallingDamage true _overlappawn)
  (Utilities|FlowControl|Sequence
    (:then_0
      (CallFunction|PlayBounceSound))
    (:then_1
      (Utilities|IsValid _overlapcontroller
        (:Is%20Valid
          (Game|Feedback|LocalPlayForceFeedback _overlapcontroller (Variables|Default|GetLaunchForceFeedback) false "TireLaunch"))
        (:Is%20Not%20Valid
          _))
      (Variables|Default|SetOverlapPawn 0)
      (Variables|Default|SetOverlapController 0)
      ;[1 1 1] Reset Bounce sound, so sound cue can be heard if "X" times bounce happens
      (Utilities|FlowControl|Delay 0.1)
      (CallFunction|ResetBounceSound))))

(event OnComponentBeginOverlap(BoxCollision_BounceTank)
  (params (OverlappedComponent :type Primitive%20Component%20Object%20Reference) (OtherActor :type Actor%20Object%20Reference) (OtherComp :type Primitive%20Component%20Object%20Reference) (OtherBodyIndex :type Integer) (bFromSweep :type Boolean) (SweepResult :type Hit%20Result%20Structure%20%28by%20ref%29))
  (CallFunction|BounceActor OverlappedComponent OtherActor SweepResult))

(event OnComponentBeginOverlap(BoxCollision_BounceBarrel)
  (params (OverlappedComponent :type Primitive%20Component%20Object%20Reference) (OtherActor :type Actor%20Object%20Reference) (OtherComp :type Primitive%20Component%20Object%20Reference) (OtherBodyIndex :type Integer) (bFromSweep :type Boolean) (SweepResult :type Hit%20Result%20Structure%20%28by%20ref%29))
  (CallFunction|BounceActor OverlappedComponent OtherActor SweepResult))

(event OnComponentBeginOverlap(BoxCollision_BounceTurret)
  (params (OverlappedComponent :type Primitive%20Component%20Object%20Reference) (OtherActor :type Actor%20Object%20Reference) (OtherComp :type Primitive%20Component%20Object%20Reference) (OtherBodyIndex :type Integer) (bFromSweep :type Boolean) (SweepResult :type Hit%20Result%20Structure%20%28by%20ref%29))
  (CallFunction|BounceActor OverlappedComponent OtherActor SweepResult))

(event OnComponentBeginOverlap(BoxCollision_BounceBarrelTop)
  (params (OverlappedComponent :type Primitive%20Component%20Object%20Reference) (OtherActor :type Actor%20Object%20Reference) (OtherComp :type Primitive%20Component%20Object%20Reference) (OtherBodyIndex :type Integer) (bFromSweep :type Boolean) (SweepResult :type Hit%20Result%20Structure%20%28by%20ref%29))
  (CallFunction|BounceActor OverlappedComponent OtherActor SweepResult))

(event Custom|TankLogging
  (params (CustomString :type String) (OverlappedComponent :type Primitive%20Component%20Object%20Reference) (LaunchedActor :type Object%20Reference) (Speed :type Float%20%28double-precision%29))
  (bind _returnvalue (Utilities|String|ToString(Float) Speed))
  (bind _returnvalue_2 (Utilities|GetDisplayName LaunchedActor))
  (bind _returnvalue_3 (Utilities|GetDisplayName OverlappedComponent))
  (bind _returnvalue_4 (Utilities|String|Append CustomString "Overlapped Component:" _returnvalue_3 ", Launched Pawn: " _returnvalue_2 ", Speed: " _returnvalue))
  (Development|PrintString _returnvalue_4)
  (Fortnite|FortShippingLog self _returnvalue_4)
  (CallFunction|CollisionLogging "Collision Info - " OverlappedComponent))

(event Custom|CollisionLogging
  (params (Custom%20String :type String) (Target :type Primitive%20Component%20Object%20Reference))
  (bind _returnvalue (Collision|GetCollisionEnabled Target))
  (bind _returnvalue_2 (Collision|GetCollisionProfileName Target))
  (bind _returnvalue_3 (Utilities|IsValid Target :!class /Script/Engine.KismetSystemLibrary))
  (bind _returnvalue_4 (Utilities|String|ToString(Boolean) _returnvalue_3))
  (bind _returnvalue_5 (Utilities|String|ToString(Name) _returnvalue_2))
  (bind _returnvalue_6 (Utilities|GetDisplayName Target))
  (bind _returnvalue_7 (Collision|GetCollisionResponsetoChannel Target "ECC_Pawn"))
  (bind _returnvalue_8 (Utilities|String|EnumtoString _returnvalue_7))
  (bind _returnvalue_9 (Utilities|String|EnumtoString _returnvalue))
  (bind _returnvalue_10 (Utilities|String|Append Custom%20String _returnvalue_6 ", IsValid: " _returnvalue_4 ", Collision Enabled: " _returnvalue_9 ", Collision Profile: " _returnvalue_5 ", Collision Response Pawn: " _returnvalue_8))
  (Development|PrintString _returnvalue_10)
  (Fortnite|FortShippingLog self _returnvalue_10))

(fn CeilingCheck
  (bind _overlappawn (Variables|Default|GetOverlapPawn)
    (:then
      (Utilities|Array|Add (LocalVariables|GetActorsToIgnore) _overlappawn)
      (Utilities|Array|AppendArray (LocalVariables|GetActorsToIgnore) (Actor|GetAttachedActors _overlappawn true false))
      (for _array_element (Actor|GetAttachedActors _overlappawn true false)
        (Utilities|Array|AppendArray (LocalVariables|GetActorsToIgnore) (Actor|GetAttachedActors _array_element true false))))
    (:else
      _))
  (bind (_outhit _returnvalue) (Collision|CapsuleTraceByChannel (Math|Vector|MakeVector (.x (Transformation|GetWorldLocation (Variables|Transformation|GetRootComponent))) (.y (Transformation|GetWorldLocation (Variables|Transformation|GetRootComponent))) (+ (.z (Transformation|GetWorldLocation (Variables|Transformation|GetRootComponent))) 200)) (Math|Vector|MakeVector (.x (Transformation|GetWorldLocation (Variables|Transformation|GetRootComponent))) (.y (Transformation|GetWorldLocation (Variables|Transformation|GetRootComponent))) (+ (.z (Transformation|GetWorldLocation (Variables|Transformation|GetRootComponent))) 340)) 50 100 "TraceTypeQuery9" false (LocalVariables|GetActorsToIgnore)))
  (return _returnvalue))

(fn CheckJumpingOrientation
  (bind (_returnvalue_x _returnvalue_y) (Math|Vector|BreakVector (Transformation|GetActorForwardVector)))
  (bind _orientationmultiplier (Variables|Default|GetOrientationMultiplier))
  (if (== (Math|Float|Round (Math|Float|Absolute(Float) _returnvalue_x)) 1)
    (return (Math|Vector|vector*float (Math|Vector|MakeVector _returnvalue_x _returnvalue_y) _orientationmultiplier))
    (else
      (return (Math|Vector|vector*float (Math|Vector|MakeVector _returnvalue_x _returnvalue_y) (* _orientationmultiplier -1.0))))))

(fn ConstructionScript
  (|CallParentFunction))
