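; Cosmetic actor for a Season 20 Week 4 narrative quest that plays a decoy tank's hit-and-deflate effect, then removes itself.
; BeginPlay activates the hit Niagara effect, plays an attached deflate sound and spawns a radial force field at the actor's root.
; The TankHit timeline animates the decoy's height, rotation and scale plus the rocks' height.
; When it finishes, the effect deactivates and the actor destroys itself after a short delay.
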(event EventBeginPlay
  (bind _sm_tankdecoy (Variables|Default|GetSM_TankDecoy))
  (Variables|Default|SetInitialPosition (Variables|/Script/Engine.SceneComponent|GetRelativeLocation _sm_tankdecoy))
  (Variables|Default|SetInitialRotation (Variables|/Script/Engine.SceneComponent|GetRelativeRotation _sm_tankdecoy))
  (Components|Activation|Activate (Variables|Default|GetNS_DummyTank_Hit) true)
  (Audio|SpawnSoundAttached "/BattlePassS20/QuestActors/Narrative/Week_04/SFX/S20_W04_Q02_TankDeflate_Cue.S20_W04_Q02_TankDeflate_Cue" (Variables|Default|GetSM_TankDecoy))
  (Game|SpawnActorfromClass "/Game/Effects/Niagara/Blueprints/PhsyicsFields/FN_RadialForce.FN_RadialForce_C" (Math|Transform|MakeTransform (Transformation|GetWorldLocation (Variables|Default|GetDefaultSceneRoot))) "Undefined" "MultiplyWithRoot" 0 4.0 4.5 1e+03 1.5 0.0 0)
  (bind (_direction _tankheight _tankrotation _rockheight _tankscale) (Timeline|TankHit Play)
    (:Update
      (Transformation|SetRelativeTransform (Variables|Default|GetSM_TankDecoy) (Math|Transform|MakeTransform (Math|Vector|MakeVector 0.0 3e+01 _tankheight) (Math|Rotator|MakeRotator (.x _tankrotation) (.y _tankrotation) (.z _tankrotation)) _tankscale) false false)
      (Transformation|SetRelativeLocationAndRotation (Variables|Default|GetSM_Rocks) (Math|Vector|MakeVector -1e+02 -5e+01 _rockheight) "1.000000,-30.000000,58.259998" false false))
    (:Finished
      (Components|Activation|Deactivate (Variables|Default|GetNS_DummyTank_Hit))
      (Utilities|FlowControl|Delay 1.0)
      (Actor|DestroyActor))))

(fn ConstructionScript)
