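; BP_S20_W04_Q3_DataDrone_Cosmetic is a cosmetic data drone actor for a Season 20 Week 4 narrative quest that animates the drone departing and then removes itself.
; On BeginPlay it plays the DroneDeparture timeline from the start.
; Each update sets the SM_Drone component's relative transform from the timeline's location and uniform scale tracks, with rotation fixed at zero.
; When the timeline finishes, the actor destroys itself.
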(event EventBeginPlay
  (bind (_direction _dronelocation _dronerotation _dronescale) (Timeline|DroneDeparture PlayFromStart)
    (:Update
      (Transformation|SetRelativeTransform (Variables|Default|GetSM_Drone) (Math|Transform|MakeTransform _dronelocation "0, 0, 0" (Math|Vector|MakeVector _dronescale _dronescale _dronescale)) false false))
    (:Finished
      (Actor|DestroyActor))))

(fn ConstructionScript)
